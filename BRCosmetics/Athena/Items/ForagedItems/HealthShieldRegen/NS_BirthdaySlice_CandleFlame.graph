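# Niagara system that renders lit candle flames on the birthday cake slice foraged health/shield item.
# Each active emitter splits particles into two spawn groups placed on the two candle mesh sockets.
# Flame and Smoke are rising curl-noise ribbons, Glow adds fading sprites, and Light loops burst-spawned light particles with a hand-keyed flicker curve.
# GlitchLarge and GlitchSmall sprite emitters on separate glitch sockets are present but disabled.
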NiagaraSystem /BRCosmetics/Athena/Items/ForagedItems/HealthShieldRegen/NS_BirthdaySlice_CandleFlame
dependencies {
 usedDataInterfaces [
  &NiagaraDataInterfaceStaticMesh
  &NiagaraDataInterfaceCurve
  &NiagaraDataInterfaceColorCurve
  &NiagaraDataInterfaceArrayColor
  &NiagaraDataInterfaceArrayFloat
  &NiagaraDataInterfaceVector2DCurve
 ]
 usedDynamicInputs [&SelectFloatFromArray, &RotateVectorByQuaternion]
 usedModules [&SystemState]
}
emitters {
 Flame {
  emitterUpdateScript [
   &EmitterState()
   &SpawnRate(SpawnRate=15, Use Spawn Probability=false)
  ]
  particleSpawnScript [
   &FN_SetSpawnGroup(Normalize Spawn Group=true, Number of Spawn Groups=2, Particles Per Group=10, Set Ribbon ID and Ribbon Link Order=true)
   &InitializeRibbon(
    Color=<rgba 10 1.75 0.2 0.3>
    Lifetime=&Multiply_Float(
     A=&Lerp_Float(
      A=0.75
      B=0.85
      Alpha=&FN_LOD_Distance_Remap(Start Distance=500, End Distance=1500, Minimum Output=0, Maximum Output=1)
     )
     B=1.25
    )
    Module.Write Color=false
    Module.Write Lifetime=true
    Module.Write Mass=true
    Module.Write Material Random=true
    Module.Write Position=true
    Module.Write RibbonTwist=false
    Module.Write RibbonWidth=true
    Ribbon Width=&Multiply_Float(
     A=&Lerp_Float(
      A=3
      B=3.5
      Alpha=&FN_LOD_Distance_Remap(Start Distance=500, End Distance=1500, Minimum Output=0, Maximum Output=1)
     )
     B=2
    )
   )
   &self:SampleStaticMeshSockets(
    SM=DataInterface() {propertyValues {
      DefaultMesh=null
      InstanceIndex=-1
      LODIndex=0
      PreviewMesh=/Game/Items/Art_noLOD/Meshes/FN_Birthday_Cake_Slice_Plate_5
      SoftSourceActor=null
      SourceMode=Default
      bAllowSamplingFromStreamingLODs=false
      bCaptureTransformsPerFrame=true
      bUsePhysicsBodyVelocity=false
      FilteredSockets [Candle_FXSocket_01, Candle_FXSocket_02]
      LODIndexUserParameter {parameter {name=None, typeDefHandle {registeredTypeIndex=108}, varData []}}
      MeshParameterBinding {parameter {name=None, typeDefHandle {registeredTypeIndex=1}, varData []}}
      SectionFilter {allowedMaterialSlots []}
     }}
    Socket:NiagaraInt32=$Particles.SpawnGroup
   ) {module=SampleStaticMeshSockets}
  ]
  particleUpdateScript [
   &ParticleState()
   &GravityForce(Gravity=<0 0 75>)
   /Niagara/Modules/Update/Forces/CurlNoiseForce(
    Noise Frequency=0.01
    Noise Strength=&FloatFromCurve(
     FloatCurve=DataInterface() {propertyValues {Curve {keys [
         {arriveTangent=0.1097, interpMode=RCIM_Cubic, leaveTangent=0.1097, tangentMode=RCTM_User, time=0, value=0}
         {arriveTangent=1.6798, interpMode=RCIM_Cubic, leaveTangent=1.6798, tangentMode=RCTM_User, time=1, value=1}
        ]}}}
     Scale Curve=30
    )
    Pan Noise=false
    Random Seed:NiagaraInt32=&RandomRangeInt(Minimum=0, Maximum=100)
   )
   &Drag(Drag=2.5, Use Linear Drag=true, Use Rotational Drag=false)
   &SolveForcesAndVelocity(Clamp Velocity=false, Limit Acceleration=false)
   &Color(Color=&ColorFromCurve(ColorCurve=DataInterface() {propertyValues {
       AlphaCurve {keys [<at 0 1 RCIM_Cubic>]}
       BlueCurve {keys [<at 0 10 RCIM_Cubic>,<at 0.1638 0.2 RCIM_Cubic>]}
       GreenCurve {keys [<at 0 1.75 RCIM_Cubic>,<at 0.1638 1.75 RCIM_Cubic>]}
       RedCurve {keys [<at 0 0.2 RCIM_Cubic>,<at 0.1638 10 RCIM_Cubic>]}
      }}))
   &ScaleColor(
    Color Value To Scale=$Particles.Color
    Scale Alpha=&FloatFromCurve(
     FloatCurve=DataInterface() {propertyValues {Curve {keys [
         {arriveTangent=-0.0176, interpMode=RCIM_Cubic, leaveTangent=-0.0176, tangentMode=RCTM_User, time=0.1, value=0}
         {arriveTangent=0.1593, interpMode=RCIM_Cubic, leaveTangent=0.1593, tangentMode=RCTM_User, time=0.3, value=1}
         {arriveTangent=-0.0405, interpMode=RCIM_Cubic, leaveTangent=-0.0405, tangentMode=RCTM_User, time=1, value=0}
        ]}}}
     Scale Curve=0.5
    )
    ScaleA=true
    ScaleRGB=true
   )
   &ScaleRibbonWidth(Ribbon Width Scale=&FloatFromCurve(
     FloatCurve=DataInterface() {propertyValues {Curve {keys [
         {arriveTangent=0.0742, interpMode=RCIM_Cubic, leaveTangent=0.0742, tangentMode=RCTM_User, time=0, value=0.4}
         {arriveTangent=-0.0613, interpMode=RCIM_Cubic, leaveTangent=-0.0613, tangentMode=RCTM_User, time=0.1408, value=0.9847}
         {arriveTangent=-0.0405, interpMode=RCIM_Cubic, leaveTangent=-0.0405, tangentMode=RCTM_User, time=1, value=0}
        ]}}}
     Scale Curve=0.5
    ))
  ]
  rendererClasses [&NiagaraRibbonRendererProperties]
  renderers [{rendererClass=&NiagaraRibbonRendererProperties}]
 }
 GlitchLarge {
  bEnabled=false
  emitterUpdateScript [
   &EmitterState()
   &SpawnRate(Use Spawn Probability=false)
  ]
  particleSpawnScript [
   &InitializeParticle(
    Color=&SelectLinearColorFromArray(
     Array Sampling Mode=::NewEnumerator0
     Direct Array Index=?
     Color Selection Array=DataInterface() {propertyValues {ColorData [<rgba 0.0467 0.4621 0.5457 0.7>, <rgba 0.5089 0.0006 0.5972 0.7>, <rgba 0.063 0.5149 0.3916 0.7>, <rgba 0.6106 0.0678 0.746 0.7>]}}
     Random Seed=0
    )
    Lifetime Min=0.5
    Lifetime Mode=::NewEnumerator1
    Mass Mode=::NewEnumerator3
    Position Mode=::NewEnumerator0
    Sprite Size=<10 5>
    Sprite Size Mode=::NewEnumerator4
    Write Material Random=true
   )
   &FN_SetSpawnGroup(Number of Spawn Groups=2)
   &self:SampleStaticMeshSockets_04(
    SM=DataInterface() {propertyValues {
      DefaultMesh=null
      InstanceIndex=-1
      LODIndex=0
      PreviewMesh=&FN_Birthday_Cake_Slice_Plate_5
      SoftSourceActor=null
      SourceMode=Default
      bAllowSamplingFromStreamingLODs=false
      bCaptureTransformsPerFrame=true
      bUsePhysicsBodyVelocity=false
      FilteredSockets [Glitch_FXSocket_01, Glitch_FXSocket_02]
      LODIndexUserParameter {parameter {name=None, typeDefHandle {registeredTypeIndex=108}, varData []}}
      MeshParameterBinding {parameter {name=None, typeDefHandle {registeredTypeIndex=1}, varData []}}
      SectionFilter {allowedMaterialSlots []}
     }}
    Socket:NiagaraInt32=$Particles.SpawnGroup
   ) {module=SampleStaticMeshSockets_04}
   &ShapeLocation(
    Box / Plane Mode:ENiagara_BoxPlaneMode=::NewEnumerator1
    MaskSpawnGroup=false
    Offset=<0 0 0.5>
    Offset Mode=::NewEnumerator0
    Override Seed=false
    Plane Midpoint:Vector2f=<0.5 0.4>
    Plane Size:Vector2f=<30 20>
    Rotation Angle Type:ENiagara_AngleInput=::NewEnumerator0
    Rotation Mode=::NewEnumerator2
    Shape Primitive=::NewEnumerator2
    Transform Order:ENiagara_TransformOrder=::NewEnumerator1
    UseOwnerScale=true
    Yaw / Pitch / Roll=&MaskVectorBySpawnGroup(Masked Vector=<107 0 90>, Passthrough Vector=<73 0 90>)
   )
  ]
  particleUpdateScript [
   &ParticleState()
   &ScaleSpriteSize(
    Non-Uniform Curve Scale=&Multiply_Vector2D_ByFloat(Float=1.5, Vector2D=<1 1>)
    Non-Uniform Curve Sprite Scale=DataInterface() {propertyValues {
      XCurve {keys [<at 0 0 RCIM_Cubic>, {arriveTangent=1, interpMode=RCIM_Cubic, leaveTangent=1, time=0.2, value=0.625}, <at 1 1 RCIM_Cubic>]}
      YCurve {keys [{tangentMode=RCTM_User, time=0, value=0.85}, {tangentMode=RCTM_User, time=1, value=1}]}
     }}
    Scale Sprite Size Mode=::NewEnumerator3
   )
   &Drag(Drag=0.25, Use Linear Drag=true, Use Rotational Drag=false)
   &ScaleColor(
    Scale Alpha=&FloatFromCurve(FloatCurve=DataInterface() {propertyValues {Curve {keys [<at 0 0 RCIM_Cubic>,<at 0.3 1 RCIM_Cubic>,<at 1 0 RCIM_Cubic>]}}}, Scale Curve=22.5)
    Scale RGB=&VectorFromFloat(Value=&FloatFromCurve(FloatCurve=DataInterface() {propertyValues {Curve {keys [<at 0 1>,<at 0.2 1.625>,<at 0.3 0.375>,<at 0.5 1.5>,<at 0.7 0.625>,<at 0.8 2>,<at 1 1>]}}}))
    ScaleA=true
    ScaleRGB=true
   )
   &SolveForcesAndVelocity(Clamp Velocity=false, Limit Acceleration=false)
   &DynamicMaterialParameters(
    Index 0 Param 1=1
    Index 0 Param 2=&RandomRangeFloat(Minimum=-1, Maximum=1)
    Index 0 Param 3=&RandomRangeFloat(Minimum=-3, Maximum=3)
    Index 0 Param 4=20
    Module.Param0WriteEnabled=true
    Module.Param1WriteEnabled=true
    Module.Param2WriteEnabled=true
    Module.Param3WriteEnabled=true
   )
   &CameraOffset(Camera Offset Amount=&RandomRangeFloat(Minimum=0, Maximum=15))
  ]
  rendererClasses [&NiagaraSpriteRendererProperties]
  renderers [{rendererClass=&NiagaraSpriteRendererProperties}]
 }
 GlitchSmall {
  bEnabled=false
  emitterUpdateScript [
   &EmitterState()
   &SpawnRate(SpawnRate=&Multiply_Float(A=2.5, B=1), Use Spawn Probability=false)
  ]
  particleSpawnScript [
   &InitializeParticle(
    Color=&SelectLinearColorFromArray(
     Array Sampling Mode=::NewEnumerator0
     Direct Array Index=?
     Color Selection Array=DataInterface() {propertyValues {ColorData [<rgba 0.0467 0.4621 0.5457 0.7>, <rgba 0.5089 0.0006 0.5972 0.7>, <rgba 0.063 0.5149 0.3916 0.7>, <rgba 0.6106 0.0678 0.746 0.7>]}}
     Random Seed=0
    )
    Lifetime Min=0.4
    Lifetime Mode=::NewEnumerator1
    Mass Mode=::NewEnumerator3
    Sprite Rotation Mode=::NewEnumerator0
    Sprite Size Max=<2.25 1.5>
    Sprite Size Min=<1 1>
    Sprite Size Mode=::NewEnumerator2
    UsePositionOffset=false
    Write Material Random=true
   )
   &FN_SetSpawnGroup(Number of Spawn Groups=2)
   &self:SampleStaticMeshSockets_05(
    SM=DataInterface() {propertyValues {
      DefaultMesh=null
      InstanceIndex=-1
      LODIndex=0
      PreviewMesh=&FN_Birthday_Cake_Slice_Plate_5
      SoftSourceActor=null
      SourceMode=Default
      bAllowSamplingFromStreamingLODs=false
      bCaptureTransformsPerFrame=true
      bUsePhysicsBodyVelocity=false
      FilteredSockets [Glitch_FXSocket_01, Glitch_FXSocket_02]
      LODIndexUserParameter {parameter {name=None, typeDefHandle {registeredTypeIndex=108}, varData []}}
      MeshParameterBinding {parameter {name=None, typeDefHandle {registeredTypeIndex=1}, varData []}}
      SectionFilter {allowedMaterialSlots []}
     }}
    Socket:NiagaraInt32=$Particles.SpawnGroup
   ) {module=SampleStaticMeshSockets_05}
   &ShapeLocation(
    Box / Plane Mode:ENiagara_BoxPlaneMode=::NewEnumerator1
    MaskSpawnGroup=false
    Offset=<0 0 0.5>
    Offset Mode=::NewEnumerator0
    Override Seed=false
    Plane Midpoint:Vector2f=<0.5 0.4>
    Plane Size:Vector2f=<30 20>
    Rotation Angle Type:ENiagara_AngleInput=::NewEnumerator0
    Rotation Mode=::NewEnumerator2
    Shape Primitive=::NewEnumerator2
    Transform Order:ENiagara_TransformOrder=::NewEnumerator1
    UseOwnerScale=true
    Yaw / Pitch / Roll=&MaskVectorBySpawnGroup(Masked Vector=<107 0 90>, Passthrough Vector=<73 0 90>)
   )
   &AddVelocity(Velocity=<0 0 1.5>, Velocity Speed Scale=&RandomRangeFloat(Minimum=1, Maximum=10))
  ]
  particleUpdateScript [
   &ParticleState()
   &Drag(Use Linear Drag=true, Use Rotational Drag=false)
   &ScaleSpriteSize(
    Uniform Curve Scale=1.5
    Uniform Curve Sprite Scale=DataInterface() {propertyValues {Curve {keys [
        <at 0 0 RCIM_Cubic>
        {arriveTangent=1, interpMode=RCIM_Cubic, leaveTangent=1, time=0.1, value=1.05}
        {arriveTangent=-1.1667, interpMode=RCIM_Cubic, leaveTangent=-1.1667, time=0.8, value=0.8}
        <at 1 0 RCIM_Cubic>
       ]}}}
   )
   &ScaleColor(
    Scale Alpha=&FloatFromCurve(FloatCurve=DataInterface() {propertyValues {Curve {keys [
         <at -0.0046 1.0113 RCIM_Cubic>
         {arriveTangent=-1.0036, interpMode=RCIM_Cubic, leaveTangent=-1.0036, time=0.7, value=0.85}
         <at 1.0011 0.002 RCIM_Cubic>
        ]}}})
    Scale RGB=&VectorFromFloat(Value=&FloatFromCurve(FloatCurve=DataInterface() {propertyValues {Curve {keys [<at 0 1>,<at 0.3 1.625>,<at 0.4 0.35>,<at 0.7 1.75>,<at 1 1>]}}}))
    ScaleA=true
    ScaleRGB=true
   )
   /SetVariables(Particles.BackingSpriteSize:Vector2f=&Vector2DFromCurve(
     Vector2Curve=DataInterface() {propertyValues {XCurve {keys [<at 0 0.2383 RCIM_Cubic>,<at 1 1.8857 RCIM_Cubic>]}, YCurve {keys [<at 0 1>,<at 1 1>]}}}
     ScaleCurve=<3 3>
    )) {isSetParameters=true}
   &SolveForcesAndVelocity(Clamp Velocity=false, Limit Acceleration=false)
   &DynamicMaterialParameters(
    Index 0 Param 1=0.25
    Index 0 Param 2=&RandomRangeFloat(Minimum=-1, Maximum=1)
    Index 0 Param 3=&RandomRangeFloat(Minimum=-3, Maximum=3)
    Index 0 Param 4=20
    Module.Param0WriteEnabled=true
    Module.Param1WriteEnabled=true
    Module.Param2WriteEnabled=true
    Module.Param3WriteEnabled=true
   )
   &CameraOffset(Camera Offset Amount=&RandomRangeFloat(Minimum=0, Maximum=15))
  ]
  rendererClasses [&NiagaraSpriteRendererProperties]
  renderers [
   {rendererClass=&NiagaraSpriteRendererProperties}
   {rendererClass=&NiagaraSpriteRendererProperties, rendererIndex=1}
  ]
 }
 Glow {
  emitterUpdateScript [
   &EmitterState()
   &SpawnRate(SpawnRate=20, Use Spawn Probability=false)
  ]
  particleSpawnScript [
   &FN_SetSpawnGroup(Normalize Spawn Group=true, Number of Spawn Groups=2, Particles Per Group=10, Set Ribbon ID and Ribbon Link Order=true)
   /Niagara/Modules/Spawn/Initialization/InitializeParticle(
    Color=<rgba 15 10 1 1>
    Lifetime=&Multiply_Float(A=&UniformRangedFloat(Minimum=0.3, Maximum=0.6), B=1)
    Module.Write Color=true
    Module.Write Lifetime=true
    Module.Write Mass=true
    Module.Write Material Random=true
    Module.Write Position=true
    Module.Write Scale=false
    Module.Write SpriteRotation=false
    Module.Write SpriteSize=true
    Position=&SimulationPosition()
    Sprite Size=&Vector2DFromFloat(Value=&Multiply_Float(A=&UniformRangedFloat(Minimum=0.65, Maximum=1), B=3))
   )
   &self:SampleStaticMeshSockets(
    SM=DataInterface() {propertyValues {
      DefaultMesh=null
      InstanceIndex=-1
      LODIndex=0
      PreviewMesh=&FN_Birthday_Cake_Slice_Plate_5
      SoftSourceActor=null
      SourceMode=Default
      bAllowSamplingFromStreamingLODs=false
      bCaptureTransformsPerFrame=true
      bUsePhysicsBodyVelocity=false
      FilteredSockets [Candle_FXSocket_01, Candle_FXSocket_02]
      LODIndexUserParameter {parameter {name=None, typeDefHandle {registeredTypeIndex=108}, varData []}}
      MeshParameterBinding {parameter {name=None, typeDefHandle {registeredTypeIndex=1}, varData []}}
      SectionFilter {allowedMaterialSlots []}
     }}
    Socket:NiagaraInt32=$Particles.SpawnGroup
   ) {module=SampleStaticMeshSockets}
   &Color(Color=&MultiplyLinearColor(
     Linear Color=$Particles.Color
     Scale Factor=/Niagara/DynamicInputs/UniformRange/UniformRangedVector4(
      Minimum:Vector4f=<1 0.25 0.25 0.01>
      Maximum:Vector4f=<1 1 1 1>
      Randomness Mode:ENiagaraRandomnessMode=::NewEnumerator0
      Random Seed:NiagaraInt32=0
      Fixed Random Seed:NiagaraBool=false
     )
    ))
  ]
  particleUpdateScript [
   &ParticleState()
   &AddVelocity(Coordinate Space=::World, Velocity=&RandomRangeVector(Minimum=<0 0 0>, Maximum=<1 1 8>))
   &Drag(Drag=30, Use Linear Drag=true, Use Rotational Drag=false)
   &SolveForcesAndVelocity(Clamp Velocity=false, Limit Acceleration=false)
   &ScaleColor(
    Scale Alpha=&FloatFromCurve(FloatCurve=DataInterface() {propertyValues {Curve {keys [<at 0.5 1>,<at 1 0>]}}})
    Scale RGB=&Lerp_Vector(A=<1 1 1>, B=<1 0 0>, Alpha=$Particles.NormalizedAge)
    ScaleA=true
    ScaleRGB=true
   )
   &ScaleSpriteSize(Scale Factor=&Vector2DFromFloat(Value=&FloatFromCurve(FloatCurve=DataInterface() {propertyValues {Curve {keys [
          {arriveTangent=-0.0036, interpMode=RCIM_Cubic, leaveTangent=-0.0036, tangentMode=RCTM_User, time=0, value=1}
          {arriveTangent=-2.2201, interpMode=RCIM_Cubic, leaveTangent=-2.2201, tangentMode=RCTM_User, time=1, value=0}
         ]}}})))
   /SetVariables(
    Particles.ColorAlt:LinearColor=&MultiplyLinearColor(Linear Color=$Particles.Color, Scale Factor=<1 0.35 0.1 0.02>)
    Particles.SpriteSizeAlt:Vector2f=&Vector2DFromFloat(Value=&FloatFromCurve(
      FloatCurve=DataInterface() {propertyValues {Curve {keys [{arriveTangent=-0.0036, interpMode=RCIM_Cubic, leaveTangent=-0.0036, tangentMode=RCTM_User, time=0, value=1}]}}}
      Scale Curve=&RandomRangeFloat(Minimum=3, Maximum=4.5)
     ))
   ) {isSetParameters=true}
   &CameraOffset(Camera Offset Amount=5)
  ]
  rendererClasses [&NiagaraSpriteRendererProperties]
  renderers [
   {rendererClass=&NiagaraSpriteRendererProperties}
   {rendererClass=&NiagaraSpriteRendererProperties, rendererIndex=1}
  ]
 }
 Light {
  emitterUpdateScript [
   &EmitterState(
    Inactive Response=::NewEnumerator1
    Life Cycle Mode=::NewEnumerator1
    Loop Behavior=::NewEnumerator1
    Loop Duration Mode=::NewEnumerator1
    UseLoopDelay=false
   )
   &SpawnBurst_Instantaneous(Spawn Count=11, Use Spawn Probability=false, UseLoopCountLimit=false)
  ]
  particleSpawnScript [
   &FN_SetSpawnGroup(Normalize Spawn Group=true, Number of Spawn Groups=2, Particles Per Group=10, Set Ribbon ID and Ribbon Link Order=true)
   /SetVariables(Particles.SpriteSize=&Vector2DFromFloat(Value=5)) {isSetParameters=true}
   /SetVariables(Particles.Lifetime=&RandomRangeFloat(Minimum=2, Maximum=3)) {isSetParameters=true}
   &self:SampleStaticMeshSockets(
    SM=DataInterface() {propertyValues {
      DefaultMesh=null
      InstanceIndex=-1
      LODIndex=0
      PreviewMesh=&FN_Birthday_Cake_Slice_Plate_5
      SoftSourceActor=null
      SourceMode=Default
      bAllowSamplingFromStreamingLODs=false
      bCaptureTransformsPerFrame=true
      bUsePhysicsBodyVelocity=false
      FilteredSockets [Candle_FXSocket_01, Candle_FXSocket_02]
      LODIndexUserParameter {parameter {name=None, typeDefHandle {registeredTypeIndex=108}, varData []}}
      MeshParameterBinding {parameter {name=None, typeDefHandle {registeredTypeIndex=1}, varData []}}
      SectionFilter {allowedMaterialSlots []}
     }}
    Socket:NiagaraInt32=$Particles.SpawnGroup
   ) {module=SampleStaticMeshSockets}
   /SetVariables(Particles.New Position:NiagaraPosition=/NiagaraFluids/DynamicInput/MaxVectorByLength(A:Vector3f=<0 0 0>, B:Vector3f=<0 0 0>)) {isSetParameters=true}
  ]
  particleUpdateScript [
   &ParticleState(Kill Particles When Lifetime Has Elapsed=false, Loop Particles Lifetime=true)
   &FN_Light_v3(
    Color=&RandomRangeLinearColor(Minimum=<rgba 7 3 0 1>, Maximum=<rgba 7 2.5 0 1>)
    Create Color Attribute=true
    Create Exponent Attribute=true
    Create Radius Attribute=true
    Enable Particle Light Enable=true
    Exponent=1
    Light Enabled=true
    Radius:NiagaraFloat=&Multiply_Float(A=&RandomRangeFloat(Minimum=18, Maximum=20), B=0.5)
    Scale Color:NiagaraFloat=&FloatFromCurve(FloatCurve=DataInterface() {propertyValues {Curve {keys [
         <at 0 1 RCIM_Cubic>
         {arriveTangent=1.4927, interpMode=RCIM_Cubic, leaveTangent=1.4927, time=0.0746, value=0.875}
         {arriveTangent=-0.0227, interpMode=RCIM_Cubic, leaveTangent=-0.0227, tangentMode=RCTM_User, time=0.1502, value=1.2241}
         {arriveTangent=-2.2781, interpMode=RCIM_Cubic, leaveTangent=-2.2781, time=0.2439, value=0.3986}
         {arriveTangent=4.6068, interpMode=RCIM_Cubic, leaveTangent=4.6068, time=0.2748, value=0.9402}
         {arriveTangent=2.2278, interpMode=RCIM_Cubic, leaveTangent=2.2278, time=0.3464, value=0.8705}
         {arriveTangent=-2.7898, interpMode=RCIM_Cubic, leaveTangent=-2.7898, time=0.3894, value=1.1955}
         {arriveTangent=-2.6422, interpMode=RCIM_Cubic, leaveTangent=-2.6422, time=0.4785, value=0.5018}
         {arriveTangent=-0.8529, interpMode=RCIM_Cubic, leaveTangent=-0.8529, time=0.4871, value=0.9373}
         {arriveTangent=0.3027, interpMode=RCIM_Cubic, leaveTangent=0.3027, time=0.5752, value=0.4194}
         {arriveTangent=4.1733, interpMode=RCIM_Cubic, leaveTangent=4.1733, time=0.6112, value=0.9748}
         {arriveTangent=2.91, interpMode=RCIM_Cubic, leaveTangent=2.91, time=0.69, value=0.8986}
         {arriveTangent=-2.2524, interpMode=RCIM_Cubic, leaveTangent=-2.2524, time=0.7356, value=1.3367}
         {arriveTangent=-3.2665, interpMode=RCIM_Cubic, leaveTangent=-3.2665, time=0.8416, value=0.5572}
         {arriveTangent=-0.8671, interpMode=RCIM_Cubic, leaveTangent=-0.8671, time=0.8675, value=0.9059}
         {arriveTangent=-0.3671, interpMode=RCIM_Cubic, leaveTangent=-0.3671, time=0.9558, value=0.4582}
         {arriveTangent=-0.6714, interpMode=RCIM_Cubic, leaveTangent=-0.6714, time=0.9899, value=0.8609}
         {arriveTangent=0.7195, interpMode=RCIM_Cubic, leaveTangent=0.7195, time=1.0792, value=0.3753}
         {arriveTangent=1.5649, interpMode=RCIM_Cubic, leaveTangent=1.5649, time=1.099, value=0.9394}
         {arriveTangent=0.3682, interpMode=RCIM_Cubic, leaveTangent=0.3682, time=1.1717, value=0.52}
         {arriveTangent=6.6831, interpMode=RCIM_Cubic, leaveTangent=6.6831, time=1.2078, value=0.9795}
         {arriveTangent=2.1336, interpMode=RCIM_Cubic, leaveTangent=2.1336, time=1.2281, value=0.8972}
         {arriveTangent=1.9914, interpMode=RCIM_Cubic, leaveTangent=1.9914, time=1.311, value=1.1996}
         {arriveTangent=0.4649, interpMode=RCIM_Cubic, leaveTangent=0.4649, time=1.3539, value=1.1478}
         {arriveTangent=-4.574, interpMode=RCIM_Cubic, leaveTangent=-4.574, time=1.396, value=1.2391}
         {arriveTangent=-0.0928, interpMode=RCIM_Cubic, leaveTangent=-0.0928, tangentMode=RCTM_User, time=1.4921, value=0.5155}
         {arriveTangent=5.2872, interpMode=RCIM_Cubic, leaveTangent=5.2872, time=1.5242, value=0.9899}
         {arriveTangent=1.3233, interpMode=RCIM_Cubic, leaveTangent=1.3233, time=1.5566, value=0.8565}
         {arriveTangent=0.3667, interpMode=RCIM_Cubic, leaveTangent=0.3667, time=1.7064, value=1.2309}
         {arriveTangent=-6.7179, interpMode=RCIM_Cubic, leaveTangent=-6.7179, time=1.7959, value=0.9442}
         {arriveTangent=-5.6448, interpMode=RCIM_Cubic, leaveTangent=-5.6448, time=1.8039, value=0.5757}
         {arriveTangent=2.6907, interpMode=RCIM_Cubic, leaveTangent=2.6907, time=1.878, value=0.4804}
         {arriveTangent=3.6859, interpMode=RCIM_Cubic, leaveTangent=3.6859, time=1.9343, value=0.9265}
         {arriveTangent=2.6815, interpMode=RCIM_Cubic, leaveTangent=2.6815, time=1.9912, value=0.8977}
         {arriveTangent=-3.7116, interpMode=RCIM_Cubic, leaveTangent=-3.7116, time=2.0308, value=1.1852}
         {arriveTangent=-1.3071, interpMode=RCIM_Cubic, leaveTangent=-1.3071, time=2.1248, value=0.402}
         {arriveTangent=0.4353, interpMode=RCIM_Cubic, leaveTangent=0.4353, time=2.2158, value=0.9434}
         {arriveTangent=-0.3447, interpMode=RCIM_Cubic, leaveTangent=-0.3447, time=2.3094, value=0.4823}
         {arriveTangent=0.6603, interpMode=RCIM_Cubic, leaveTangent=0.6603, time=2.3258, value=0.9055}
         {arriveTangent=0.3801, interpMode=RCIM_Cubic, leaveTangent=0.3801, time=2.4391, value=0.568}
         {arriveTangent=4.4872, interpMode=RCIM_Cubic, leaveTangent=4.4872, time=2.4654, value=0.9586}
         {arriveTangent=3.4107, interpMode=RCIM_Cubic, leaveTangent=3.4107, time=2.5143, value=0.9055}
         {arriveTangent=-0.515, interpMode=RCIM_Cubic, leaveTangent=-0.515, time=2.5385, value=1.2081}
         {arriveTangent=-2.6117, interpMode=RCIM_Cubic, leaveTangent=-2.6117, time=2.597, value=0.8629}
         {arriveTangent=-2.5605, interpMode=RCIM_Cubic, leaveTangent=-2.5605, time=2.6456, value=0.9286}
         {arriveTangent=0.1009, interpMode=RCIM_Cubic, leaveTangent=0.1009, time=2.7341, value=0.5119}
         {arriveTangent=0.0929, interpMode=RCIM_Cubic, leaveTangent=0.0929, time=2.7509, value=0.9392}
         {arriveTangent=0.2939, interpMode=RCIM_Cubic, leaveTangent=0.2939, time=2.9074, value=0.528}
         {arriveTangent=8.4131, interpMode=RCIM_Cubic, leaveTangent=8.4131, time=2.9271, value=0.991}
         {arriveTangent=0.1236, interpMode=RCIM_Cubic, leaveTangent=0.1236, time=2.957, value=0.9449}
         <at 3 1 RCIM_Cubic>
        ]}}})
   )
  ]
  rendererClasses [&NiagaraLightRendererProperties]
  renderers [{rendererClass=&NiagaraLightRendererProperties, rendererIndex=0}]
 }
 Smoke {
  emitterUpdateScript [
   &EmitterState()
   &SpawnRate(SpawnRate=15, Use Spawn Probability=false)
  ]
  particleSpawnScript [
   &FN_SetSpawnGroup(Normalize Spawn Group=true, Number of Spawn Groups=2, Particles Per Group=10, Set Ribbon ID and Ribbon Link Order=true)
   &InitializeRibbon(
    Color=<rgba 10 1.75 0.2 1>
    Lifetime=&Multiply_Float(
     A=&RandomRangeFloat(Minimum=1.25, Maximum=1.75, Randomness Mode=::NewEnumerator1, Random Seed=$Particles.SpawnGroup)
     B=1.2
    )
    Module.Write Color=true
    Module.Write Lifetime=true
    Module.Write Mass=true
    Module.Write Material Random=true
    Module.Write Position=true
    Module.Write RibbonTwist=false
    Module.Write RibbonWidth=true
    Ribbon Width=&Multiply_Float(A=&RandomRangeFloat(Minimum=2.5, Maximum=3.5), B=1.25)
   )
   &self:SampleStaticMeshSockets(
    SM=DataInterface() {propertyValues {
      DefaultMesh=null
      InstanceIndex=-1
      LODIndex=0
      PreviewMesh=&FN_Birthday_Cake_Slice_Plate_5
      SoftSourceActor=null
      SourceMode=Default
      bAllowSamplingFromStreamingLODs=false
      bCaptureTransformsPerFrame=true
      bUsePhysicsBodyVelocity=false
      FilteredSockets [Candle_FXSocket_01, Candle_FXSocket_02]
      LODIndexUserParameter {parameter {name=None, typeDefHandle {registeredTypeIndex=108}, varData []}}
      MeshParameterBinding {parameter {name=None, typeDefHandle {registeredTypeIndex=1}, varData []}}
      SectionFilter {allowedMaterialSlots []}
     }}
    Socket:NiagaraInt32=$Particles.SpawnGroup
   ) {module=SampleStaticMeshSockets}
  ]
  particleUpdateScript [
   &ParticleState()
   &GravityForce(Gravity=<0 0 75>)
   &CurlNoiseForce(
    Noise Frequency=0.01
    Noise Strength=&FloatFromCurve(
     FloatCurve=DataInterface() {propertyValues {Curve {keys [
         {arriveTangent=1.0245, interpMode=RCIM_Cubic, leaveTangent=1.0245, tangentMode=RCTM_User, time=0, value=0}
         {arriveTangent=1.6798, interpMode=RCIM_Cubic, leaveTangent=1.6798, tangentMode=RCTM_User, time=1, value=1}
        ]}}}
     Scale Curve=30
    )
    Pan Noise=false
    Random Seed:NiagaraInt32=&RandomRangeInt(Minimum=0, Maximum=100)
   )
   &Drag(Drag=2.5, Use Linear Drag=true, Use Rotational Drag=false) {module=Drag001}
   &FN_RibbonMaterial(Color Pan U Speed=-0.5, Erosion Amount=1, Macro Distortion Amount=1, Opacity Erosion=0.25, Opacity Pan U Speed=-0.5, Opacity Power=0.35)
   &ScaleRibbonWidth(Ribbon Width Scale=&FloatFromCurve(
     FloatCurve=DataInterface() {propertyValues {Curve {keys [
         {arriveTangent=0.0219, interpMode=RCIM_Cubic, leaveTangent=0.0219, tangentMode=RCTM_User, time=0, value=0}
         <at 0.5 0.9666 RCIM_Cubic>
         {arriveTangent=-0.2116, interpMode=RCIM_Cubic, leaveTangent=-0.2116, tangentMode=RCTM_User, time=1, value=0}
        ]}}}
     Scale Curve=&RandomRangeFloat(Minimum=1, Maximum=1)
    )) {module=ScaleRibbonWidth001}
   &SolveForcesAndVelocity(Clamp Velocity=false, Limit Acceleration=false)
   &ScaleColor(
    Scale Alpha=&FloatFromCurve(
     FloatCurve=DataInterface() {propertyValues {Curve {keys [
         {arriveTangent=9.0725, interpMode=RCIM_Cubic, leaveTangent=9.0725, tangentMode=RCTM_User, time=0, value=0}
         {arriveTangent=0.0486, interpMode=RCIM_Cubic, leaveTangent=0.0486, tangentMode=RCTM_User, time=0.1408, value=0.9847}
         {arriveTangent=-0.0405, interpMode=RCIM_Cubic, leaveTangent=-0.0405, tangentMode=RCTM_User, time=1, value=0}
        ]}}}
     Scale Curve=1.25
    )
    ScaleA=true
    ScaleRGB=true
   )
  ]
  rendererClasses [&NiagaraRibbonRendererProperties]
  renderers [{rendererClass=&NiagaraRibbonRendererProperties}]
 }
}
summary {
 systemName=NS_BirthdaySlice_CandleFlame
 emitters [
  {emitterName=Glow, rendererClasses [&NiagaraSpriteRendererProperties]}
  {emitterName=Flame, rendererClasses [&NiagaraRibbonRendererProperties]}
  {emitterName=Light, rendererClasses [&NiagaraLightRendererProperties]}
  {emitterName=Smoke, rendererClasses [&NiagaraRibbonRendererProperties]}
  {bEnabled=false, emitterName=GlitchLarge, rendererClasses [&NiagaraSpriteRendererProperties]}
  {bEnabled=false, emitterName=GlitchSmall, rendererClasses [&NiagaraSpriteRendererProperties]}
 ]
 userVariables []
}
